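# BuffCut_AnimBP_Child is a child Animation Blueprint targeting the skeleton of the BuffCut quadruped cosmetic companion.
# It adds no variables or animation logic of its own, and its update event only forwards the delta time to the parent implementation.
# AllowsCinematicOverride returns false, and an unconnected TryGetPawnOwner node sits unused in the event graph.
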AnimBlueprint /CosmeticCompanions/Assets/Quadruped/BuffCut/Blueprint/BuffCut_AnimBP_Child
target_skeleton=/CosmeticCompanions/Assets/Quadruped/BuffCut/Meshes/BuffCut_Quadruped_Skeleton
graphs [
 AllowsCinematicOverride {nodes [
   AllowsCinematicOverride() {is=title K2Node_FunctionEntry}
   > Return Node!(execute, ReturnValue:bool=false)
  ]}
 EventGraph {nodes [
   n1 = Event BlueprintUpdateAnimation()
   > Parent: BlueprintUpdateAnimation!(execute, DeltaTimeX:float=@n1.DeltaTimeX) -> (then) {is=title K2Node_CallParentFunction}
   TryGetPawnOwner()
  ]}
]
variables []
